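; Temporary NitroDrome reactive fire actor with a flame mesh, embers and point light.
; Construction and BeginPlay run a flame toggle that, per DisableFlames, hides embers and divides the material's emissive tint by six, or shows embers at full tint.
; Replicated CheatTestCosmetic triggers a timeline that scales the fire mesh on non-server machines.
; Construction also caches the light's attenuation radius and shrinks it on Switch.
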(event Custom|NitroDromeForceCosmeticTestEvent
  (Variables|Default|SetCheatTestCosmetic false))

(event Custom|ScaleFlame
  (Variables|Default|SetBaseScale (Transformation|GetActorScale3D self))
  (if (not (Networking|IsServer))
    (bind (_direction _newtrack_0) (Timeline|Timeline PlayFromStart)
      (:Update
        (Transformation|SetWorldScale3D (Variables|Default|GetFire_Static) (* _newtrack_0 (Variables|Default|GetBaseScale)))))))

(event EventBeginPlay
  (CallFunction|Disable/EnableFlames))

(event Collision|EventActorBeginOverlap
  (params (OtherActor :type Actor%20Object%20Reference)))

(event EventTick
  (params (DeltaSeconds :type Float%20%28single-precision%29)))

(fn Disable/EnableFlames
  (bind _newvar (Variables|Default|GetNewVar))
  (bind _returnvalue (Math|Conversions|ToVector(LinearColor) _newvar))
  (bind _returnvalue_2 (Rendering|Material|GetVectorParameterValue (Class|BTechArtGlobalLibrary|GetDynamicMaterialInstance (Variables|Debug|GetStaticMeshComponent)) "SFX_Emissive_Tint(Expand)" :!class /Script/Engine.MaterialInstanceDynamic))
  (Variables|Default|SetNewVar _returnvalue_2)
  (if (Variables|Building|TimeOfDayLights|GetDisableFlames)
    (Rendering|SetVisibility (Variables|Default|GetEmbers) false false)
    (Rendering|Material|SetVectorParameterValueonMaterials (Variables|Debug|GetStaticMeshComponent) "SFX_Emissive_Tint(Expand)" (/ _returnvalue "6.000000,6.000000,6.000000"))
    (else
      (Rendering|SetVisibility (Variables|Default|GetEmbers) true false)
      (Rendering|Material|SetVectorParameterValueonMaterials (Variables|Debug|GetStaticMeshComponent) "SFX_Emissive_Tint(Expand)" _returnvalue))))

(fn OnRep_CheatTestCosmetic
  (CallFunction|ScaleFlame))

(fn ConstructionScript
  (bind _pointlight (Variables|Default|GetPointLight))
  (bind _pointlight_2 (Variables|Default|GetPointLight))
  (bind _attenuationradius (Variables|/Script/Engine.LocalLightComponent|GetAttenuationRadius _pointlight_2))
  (|CallParentFunction)
  (CallFunction|Disable/EnableFlames)
  (Variables|Default|SetBaseAttenuation _attenuationradius)
  (if (TouchPlatform|IsNintendoSwitch)
    (Rendering|Lighting|SetAttenuationRadius _pointlight (* _attenuationradius 0.8))
    (else
      (Rendering|Lighting|SetAttenuationRadius _pointlight _attenuationradius))))
